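; Season 14 quest gameplay ability that credits the player for keeping health and shields at a threshold for a stage-based duration.
; On the server, attribute-threshold tasks set success flags, with a delayed re-check to ignore brief spikes.
; Success starts a timer sized by the active quest stage, whose expiry sends the quest stat event and ends the ability.
; Pausing only sets a flag; resume and timer-end logic are empty here.
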(event Custom|DetermineStage
  (bind _owningplayerquestmanager (Variables|GAVars|Data|GetOwningPlayerQuestManager))
  ;[1 1 1] Set the quest time based on which quest stage is active
  (Utilities|FlowControl|DoOnce false
    (:Completed
      (if (Utilities|Enum|Equal(Enum) (Quest|GetFortQuestState (Quests|GetQuestwithDefinition _owningplayerquestmanager 0)) "Active")
        (Variables|Default|SetQuestTime 1.2e+02)
        (elif (Utilities|Enum|Equal(Enum) (Quest|GetFortQuestState (Quests|GetQuestwithDefinition _owningplayerquestmanager 0)) "Active")
          (Variables|Default|SetQuestTime 2.4e+02)
          (elif (Utilities|Enum|Equal(Enum) (Quest|GetFortQuestState (Quests|GetQuestwithDefinition _owningplayerquestmanager 0)) "Active")
            (Variables|Default|SetQuestTime 4.8e+02)))))))

(event Custom|ContinueMonitoring
  ;[1 1 1] Wait for attribute changes and check progress again
  (bind (_asynctaskproxy _bmatchescomparison) (Ability|Tasks|WaitForAttributeChangeThreshold (Variables|Default|GetAtt_Health) "GreaterThanOrEqualTo" (Variables|Default|GetComparisonValue))
    (:then
      (bind (_asynctaskproxy_2 _bmatchescomparison_2) (Ability|Tasks|WaitForAttributeChangeThreshold (Variables|Default|GetAtt_Shield) "GreaterThanOrEqualTo" (Variables|Default|GetComparisonValue))
        (:OnChange
          (Variables|Default|SetShieldSuccess _bmatchescomparison_2))))
    (:OnChange
      (Variables|Default|SetHealthSuccess _bmatchescomparison)))
  (CallFunction|TryProgressQuest)
  (CallFunction|DetermineStage))

(event Ability|EventActivateAbility
  ;[1 1 1] Check starting health and shield values in case player spawns with threshold success
  (|CallParentFunction)
  (if (Networking|IsServer)
    ;[1 1 1] Wait for attribute changes and check progress again
    (bind (_asynctaskproxy _bmatchescomparison) (Ability|Tasks|WaitForAttributeChangeThreshold (Variables|Default|GetAtt_Health) "GreaterThanOrEqualTo" (Variables|Default|GetComparisonValue))
      (:then
        (bind (_asynctaskproxy_2 _bmatchescomparison_2) (Ability|Tasks|WaitForAttributeChangeThreshold (Variables|Default|GetAtt_Shield) "GreaterThanOrEqualTo" (Variables|Default|GetComparisonValue))
          (:OnChange
            (Variables|Default|SetShieldSuccess _bmatchescomparison_2))))
      (:OnChange
        (Variables|Default|SetHealthSuccess _bmatchescomparison)))
    (CallFunction|TryProgressQuest)
    (CallFunction|DetermineStage)))

(event Custom|TryProgressQuest
  (bind (_bsuccessfullyfoundattribute _returnvalue) (Ability|Attribute|GetFloatAttribute (Variables|GAVars|Data|GetOwningPlayerPawn) (Variables|Default|GetAtt_Health)))
  (bind (_bsuccessfullyfoundattribute_2 _returnvalue_2) (Ability|Attribute|GetFloatAttribute (Variables|GAVars|Data|GetOwningPlayerPawn) (Variables|Default|GetAtt_Shield)))
  ;[1 1 1] If health and shield meet thresholds, send quest progress
  (Utilities|FlowControl|DoOnce false
    (:Completed
      ;;[1 0.8959 0.3865] Added this delay and additional check to prevent future occurrences of FORT-251834 In that case, the LTM flashed the HP and Shield to 100 briefly when the player jumped from the bus
      (if (and (Variables|Default|GetHealthSuccess) (Variables|Default|GetShieldSuccess))
        (Utilities|FlowControl|Delay 0.05)
        (if (and (>= _returnvalue (Variables|Default|GetComparisonValue)) (>= _returnvalue_2 (Variables|Default|GetComparisonValue)))
          (if (Variables|Default|GetCountdownStarted)
            (CallFunction|ResumeCountdownTimer)
            (else
              (CallFunction|StartCountdownTimer)))
          (else
            (CallFunction|PauseCountdownTimer)))
        (else
          (CallFunction|PauseCountdownTimer))))))

(event Custom|ResetCheck
  (bind (_bsuccessfullyfoundattribute _returnvalue) (Ability|Attribute|GetFloatAttribute (Variables|GAVars|Data|GetOwningPlayerPawn) (Variables|Default|GetAtt_Health)))
  (bind (_bsuccessfullyfoundattribute_2 _returnvalue_2) (Ability|Attribute|GetFloatAttribute (Variables|GAVars|Data|GetOwningPlayerPawn) (Variables|Default|GetAtt_Shield)))
  ;[1 1 1] If health and shield meet thresholds, send quest progress
  (if (not (Variables|Default|GetQuestComplete))
    (Utilities|FlowControl|DoOnce false
      (:Completed
        ;;[1 0.8959 0.3865] Added this delay and additional check to prevent future occurrences of FORT-251834 In that case, the LTM flashed the HP and Shield to 100 briefly when the player jumped from the bus
        (if (and (Variables|Default|GetHealthSuccess) (Variables|Default|GetShieldSuccess))
          (Utilities|FlowControl|Delay 0.05)
          (if (and (>= _returnvalue (Variables|Default|GetComparisonValue)) (>= _returnvalue_2 (Variables|Default|GetComparisonValue)))
            (if (Variables|Default|GetCountdownStarted)
              (CallFunction|ResumeCountdownTimer)
              (else
                (CallFunction|StartCountdownTimer)))
            (else
              (CallFunction|PauseCountdownTimer)))
          (else
            (CallFunction|PauseCountdownTimer)))))))

(event Custom|EndTimer)

(event Custom|FinishQuest
  ;[1 1 1] Start a timer when the player reaches max health/shields, pause and hide the timer
  ; when their health or shields fall below max, bring back the timer when their health or
  ; shields return to max
  (Quest|SendComplexCustomStatEvent (Variables|GAVars|Data|GetOwningPlayerQuestManager) 0 "(GameplayTags=((TagName=\"Athena.Quests.TimedFullHealthShields\")))" "(GameplayTags=)" 1)
  (Ability|EndAbility))

(event Custom|WinCondition
  ;[1 1 1] Start a timer when the player reaches max health/shields, pause and hide the timer
  ; when their health or shields fall below max, bring back the timer when their health or
  ; shields return to max
  (CallFunction|FinishQuest)
  (Variables|Default|SetQuestComplete true)
  (CallFunction|EndTimer))

(event Custom|StartCountdownTimer
  ;[1 1 1] Start a timer when the player reaches max health/shields, pause and hide the timer
  ; when their health or shields fall below max, bring back the timer when their health or
  ; shields return to max
  (Variables|Default|SetCountdownStarted true)
  (bind _returnvalue (Utilities|Time|SetTimerbyEvent (delegate Custom|WinCondition) (Variables|Default|GetQuestTime)))
  (Utilities|FlowControl|Delay 1.0)
  (CallFunction|ContinueMonitoring)
  (CallFunction|ResetCheck))

(event Ability|EventFailedToActivatePassiveAbility
  (params (ActorInfo :type Gameplay%20Ability%20Actor%20Info%20Structure))
  (|CallParentFunction))

(event Custom|ResumeCountdownTimer
  ;[1 1 1] Start a timer when the player reaches max health/shields, pause and hide the timer
  ; when their health or shields fall below max, bring back the timer when their health or
  ; shields return to max
  (if (Variables|Default|GetCountdownPaused)
    _))

(event Custom|PauseCountdownTimer
  ;[1 1 1] Start a timer when the player reaches max health/shields, pause and hide the timer
  ; when their health or shields fall below max, bring back the timer when their health or
  ; shields return to max
  (Variables|Default|SetCountdownPaused true))

(event Ability|EventOnEndAbility
  (params (bWasCancelled :type Boolean)))
